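# Sound cue for the Biker pickaxe's enemy impact, a cosmetic melee hit sound.
# A short fixed delay feeds a weighted random pick among six impact waves, with one variant weighted to play less often.
# It primes on load and routes through the impacts sound class, shared pickaxe attenuation and property sheet, and a player melee weapon concurrency set.
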SoundCue /BRCosmetics/Sounds/PickAxes/Biker/PickAxe_Biker_EnemyImpact_Cue
root=Delay_1
nodes [
 {id=Delay_1, children [Random_0], properties {delayMax=0.1, delayMin=0.1}}
 {
  id=Random_0
  children [WavePlayer_0, WavePlayer_2, WavePlayer_8, WavePlayer_9, WavePlayer_10, WavePlayer_11]
  properties {weights [1,1,1,0.2,1,1]}
 }
 {id=WavePlayer_0, properties {soundWaveAssetPtr=&/PickAxe_Biker_EnemyImpact_01}}
 {id=WavePlayer_10, properties {soundWaveAssetPtr=&/PickAxe_Biker_EnemyImpact_07}}
 {id=WavePlayer_11, properties {soundWaveAssetPtr=&/PickAxe_Biker_EnemyImpact_08}}
 {id=WavePlayer_2, properties {soundWaveAssetPtr=&/PickAxe_Biker_EnemyImpact_03}}
 {id=WavePlayer_8, properties {soundWaveAssetPtr=&/PickAxe_Biker_EnemyImpact_05}}
 {id=WavePlayer_9, properties {soundWaveAssetPtr=&/PickAxe_Biker_EnemyImpact_06}}
]
properties {
 attenuationSettings=/Game/Sounds/Attenuation/Fort_Pickaxe_Impact_Att
 audioPropertiesSheet=/BRCosmetics/Sounds/Settings/PropertySheets/Pickaxe/APS_Cosmetic_Pickaxe_Impact
 bPrimeOnLoad=true
 maxDistance=4300
 soundClassObject=&Fort_Impacts
 volumeMultiplier=0.6
 attenuationOverrides {
  bAttenuateWithLPF=true
  distanceAlgorithm=Logarithmic
  falloffDistance=800
  lPFRadiusMax=800
  lPFRadiusMin=400
  attenuationShapeExtents {x=300}
 }
 concurrencyOverrides {bLimitToOwner=true, maxCount=2, resolutionRule=StopOldest}
 concurrencySet [&Player_Melee_Weapon_Conc]
}
